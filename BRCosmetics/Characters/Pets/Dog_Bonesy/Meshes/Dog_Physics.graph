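# Physics asset for the Bonesy dog pet's skeletal mesh, covering only the upper spine, neck and head.
# Four capsule bodies sit on spine_02, neck_01, neck_02 and head, with a larger rotated capsule on the spine and small ones on the rest.
# Three constraints link them in a chain from spine to head, each with identical limited swing and a tighter twist limit.
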PhysicsAsset /BRCosmetics/Characters/Pets/Dog_Bonesy/Meshes/Dog_Physics
bodies {
 head [{shapeType=Capsule, length=2, radius=2}]
 neck_01 [{shapeType=Capsule, length=2, radius=2}]
 neck_02 [{shapeType=Capsule, length=2, radius=2}]
 spine_02 [{shapeType=Capsule, length=6.8723, radius=4.7383, center=<0.7694 1.5898 0>, rotation=<rot 0 -90 0>}]
}
constraints [
 {bone1Name=neck_02, bone2Name=neck_01, swing1LimitDegrees=45, swing2LimitDegrees=45, twistLimitDegrees=15}
 {bone1Name=head, bone2Name=neck_02, swing1LimitDegrees=45, swing2LimitDegrees=45, twistLimitDegrees=15}
 {bone1Name=neck_01, bone2Name=spine_02, swing1LimitDegrees=45, swing2LimitDegrees=45, twistLimitDegrees=15}
]
